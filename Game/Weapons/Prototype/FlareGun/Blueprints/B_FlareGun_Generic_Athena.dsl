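; Blueprint for the generic Athena flare gun weapon, in the prototype weapons folder.
; Its muzzle flash FX event calls the parent implementation, passing false, then toggles the active state of the rear muzzle component.
; The construction script only calls the parent version, leaving other behavior to the parent class.
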(event EventMuzzleFlashFX
  (params (Persistent%20Fire :type Boolean))
  (|CallParentFunction false)
  (Components|Activation|ToggleActive (Variables|Default|GetRearMuzzle)))

(fn ConstructionScript
  (|CallParentFunction))
