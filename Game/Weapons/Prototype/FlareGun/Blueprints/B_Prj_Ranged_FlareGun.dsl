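; Prototype flare gun projectile that explodes on stop, on bounce or when its fuse timer expires, leaving a spray decal of the owner's banner.
; BeginPlay starts the fuse and trail and async-loads the banner icon and colors; Tick spins mesh and trail cosmetically.
; Stop, bounce and fuse record the stop location before DoExplosion.
; OnExploded spawns the decal and explosion effects, with a water-surface branch adding camera shake, force feedback and post-process fade.
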(event EventBeginPlay
  (bind _returnvalue (BlueprintContext|GetFortMcpContext))
  ;[1 1 1] Begin Play
  (bind _returnvalue_2 (Effects|SpawnEmitteratLocation (Variables|Default|GetRocket%20Trail%20Template) (Transformation|GetWorldLocation (Variables|Projectile_Rocket|GetMesh)) (Transformation|GetWorldRotation (Variables|Projectile_Rocket|GetMesh))))
  (Variables|Effects|SetRocketTrailPS _returnvalue_2)
  (Transformation|AddLocalRotation (Variables|Effects|GetRocketTrailPS) (Math|Rotator|MakeRotator (Math|Random|RandomFloatinRange 0.0 360)) false false)
  (Utilities|Time|SetTimerbyFunctionName self "FuseTimerMax" (Variables|Default|GetFlareFuseInSecs))
  (Networking|CanExecuteCosmeticEvents
    (:True
      (Utilities|FlowControl|Delay 0)
      (bind _returnvalue_3 (PlayerInfo|GetOwnerPlayerController))
      (bind _returnvalue_4 (Banner|GetBannerLargeIconAssetfromRowName (Utilities|String|StringToName (CosmeticLoadout|GetBannerIconIdfromLoadout (MCPContext|GetCosmeticLoadoutforPlayer _returnvalue (Default|GetGameAccountId _returnvalue_3)))) "Spray"))
      (Utilities|FlowControl|Sequence
        (:then_0
          (bind _object (Utilities|AsyncLoadAsset _returnvalue_4)
            (:Completed
              (bind _astexture_2d (Utilities|Casting|CastToTexture2D _object))
              (Variables|Default|SetBannerIconLarge _astexture_2d))))
        (:then_1
          (bind _returnvalue_5 (Banner|GetBannerColorfromRowName (Utilities|String|StringToName (CosmeticLoadout|GetBannerColorIdfromLoadout (MCPContext|GetCosmeticLoadoutforPlayer _returnvalue (Default|GetGameAccountId _returnvalue_3))))))
          (Variables|Default|SetBannerPrimaryColor (Banner|GetPrimaryColor _returnvalue_5))
          (Variables|Default|SetBannerSecondaryColor (Banner|GetSecondaryColor _returnvalue_5)))))))

(event EventTick
  (params (DeltaSeconds :type Float%20%28single-precision%29))
  ;[1 1 1] On Tick, spin the rocket and particle trail
  (Variables|Default|SetTick%20Delta DeltaSeconds)
  (Networking|CanExecuteCosmeticEvents
    (:True
      (if (Components|Activation|IsActive (Variables|Effects|GetRocketTrailPS))
        (Transformation|SetWorldLocation (Variables|Effects|GetRocketTrailPS) (Transformation|GetWorldLocation (Variables|Projectile_Rocket|GetMesh)) false false)
        (Variables|Effects|SetRocketSpin (Math|Rotator|CombineRotators (Math|Rotator|MakeRotator (* (Variables|Default|GetTick%20Delta) (Variables|Effects|GetRocketSpinRate))) (Variables|Effects|GetRocketSpin)))
        (Transformation|AddLocalRotation (Variables|Projectile_Rocket|GetMesh) (Variables|Effects|GetRocketSpin) false false)
        (Transformation|AddLocalRotation (Variables|Effects|GetRocketTrailPS) (Variables|Effects|GetRocketSpin) false false)))))

(event EventOnStop
  (params (Hit :type Hit%20Result%20Structure%20%28by%20ref%29))
  (bind (_bblockinghit _binitialoverlap _time _distance _location _impactpoint _normal _impactnormal _physmat) (Collision|BreakHitResult Hit))
  ;[1 1 1] Impact and Explosion Stuff
  (Variables|Effects|SetDecalLocation _impactpoint)
  (Variables|Effects|SetStopLocation (+ _impactpoint (* _impactnormal (Variables|Effects|GetExplosion%20Impact%20Offset))))
  (Variables|Effects|SetStopRotZ (Math|Rotator|MakeRotfromZ _impactnormal))
  (Variables|Default|SetSurfaceType (Variables|/Script/PhysicsCore.PhysicalMaterial|GetSurfaceType _physmat))
  (if (not (Delivery|IsBeingKilled))
    (Delivery|DoExplosion)))

(event Custom|FuseTimerMax
  ;[1 1 1] Impact and Explosion Stuff
  (Variables|Effects|SetTimerMaxReached? true)
  (Variables|Effects|SetStopLocation (Transformation|GetWorldLocation (Variables|Projectile_Rocket|GetMesh)))
  (if (not (Delivery|IsBeingKilled))
    (Delivery|DoExplosion)))

(event EventOnBounce
  (params (Hit :type Hit%20Result%20Structure%20%28by%20ref%29))
  ;[1 1 1] Impact and Explosion Stuff
  (Variables|Effects|SetTimerMaxReached? true)
  (Variables|Effects|SetStopLocation (Transformation|GetWorldLocation (Variables|Projectile_Rocket|GetMesh)))
  (if (not (Delivery|IsBeingKilled))
    (Delivery|DoExplosion)))

(event EventOnExploded
  (params (HitActors :type Array%20of%20Actor%20Object%20References) (HitResults :type Array%20of%20Hit%20Result%20Structures))
  (bind _stoplocation (Variables|Effects|GetStopLocation))
  (bind _stoplocation_2 (Variables|Effects|GetStopLocation))
  (bind _stoprotz (Variables|Effects|GetStopRotZ))
  ;[1 1 1] Impact and Explosion Stuff
  (Networking|CanExecuteCosmeticEvents
    (:True
      (Development|PrintString "Boom")
      (Components|Activation|SetActive (Variables|Effects|GetRocketTrailPS) false false)
      (Utilities|FlowControl|Sequence
        (:then_0
          (Game|SpawnActorfromClass "/Game/Weapons/Prototype/FlareGun/Blueprints/B_SprayDecalTest_Flare.B_SprayDecalTest_Flare_C" (Transformation|GetActorTransform) "Undefined" "OverrideRootScale" 0 (Variables|Default|GetBannerSmokeMaterial) (Variables|Effects|GetDecal%20Size%20Min) (Variables|Effects|GetDecal%20Size%20Max) (Variables|Default|GetBannerPrimaryColor) 32.0 2.0 1.0 1.0 0.4 0.1 (Variables|Default|GetBannerIconLarge) 0))
        (:then_1
          ;;[1 1 1] SingleLineCheckForWorldObjectsSurfaceType
          (switch Utilities|FlowControl|Switch|SwitchonEPhysicalSurface (Variables|Default|GetSurfaceType)
            (:SurfaceType_Default
              _)
            (:SurfaceType1
              _)
            (:SurfaceType2
              _)
            (:SurfaceType3
              _)
            (:SurfaceType4
              _)
            (:SurfaceType5
              _)
            (:SurfaceType6
              _)
            (:SurfaceType7
              _)
            (:SurfaceType8
              (Effects|SpawnEmitteratLocation (Variables|Effects|GetWater_Explosion_PS) _stoplocation _stoprotz)
              (Audio|PlaySoundatLocation (Variables|Sounds|GetRocket_Projectile_Explosion_Water_Sound) (Variables|Effects|GetStopLocation))
              (Rendering|SetVisibility (Variables|Projectile_Rocket|GetMesh) false false)
              ;[1 1 1] Shake that camera / Post Process
              (Utilities|FlowControl|Sequence
                (:then_0
                  (if (<= (Math|Vector|VectorLength (- (Variables|Effects|GetStopLocation) (Transformation|GetWorldLocation (Variables|/Script/Engine.PlayerCameraManager|GetTransformComponent (Game|GetPlayerCameraManager 0))))) (Variables|/Script/Engine.PostProcessComponent|GetBlendRadius (Variables|Default|GetPP_Explode)))
                    (bind (_direction _ppfade) (Timeline|PPFader PlayFromStart)
                      (:Update
                        (Variables|/Script/Engine.PostProcessComponent|SetBlendWeight _ppfade (Variables|Default|GetPP_Explode))))))
                (:then_1
                  (Camera|PlayWorldCameraShake "/Game/Weapons/FORT_RocketLaunchers/Blueprints/B_RocketLauncherCameraShake.B_RocketLauncherCameraShake_C" _stoplocation_2 400 850)
                  (Audio|Components|Audio|FadeOut (Variables|Default|GetRocket_AudioComponent) 1 0.0 "Linear" :!class /Script/Engine.AudioComponent)
                  (Fortnite|PlayLocalForceFeedbackatLocationMulti (Variables|Default|GetExplosionForceFeedbackNear) (Variables|Default|GetExplosionForceFeedbackFar) _stoplocation_2 400 850 "Rocket"))))
            (:SurfaceType9
              _)
            (:SurfaceType10
              _)
            (:SurfaceType11
              _)
            (:SurfaceType12
              _)
            (:SurfaceType13
              _)
            (:SurfaceType14
              _)
            (:SurfaceType15
              _)
            (:SurfaceType16
              _)
            (:SurfaceType17
              _))
          (Effects|SpawnEmitteratLocation (Variables|Effects|GetExplosion_Generic_PS) _stoplocation _stoprotz)
          (Audio|PlaySoundatLocation (Variables|Sounds|GetRocket_Projectile_Explosion_Sound) (Variables|Effects|GetStopLocation)))))))

(event EventDestroyed
  ;[1 1 1] Impact and Explosion Stuff
  (Networking|CanExecuteCosmeticEvents
    (:True
      (Components|Activation|SetActive (Variables|Effects|GetRocketTrailPS) false false))))

(fn ConstructionScript
  (Effects|Components|ParticleSystem|SetTemplate (Variables|Default|GetRibbon_Trail_PSC) (Variables|Effects|GetRibbon_Trail_PS))
  (Effects|Components|ParticleSystem|SetTemplate (Variables|Default|GetWater_Explosion_PSC) (Variables|Effects|GetWater_Explosion_PS))
  (Effects|Components|ParticleSystem|SetTemplate (Variables|Default|GetExplosion_Generic_PSC) (Variables|Effects|GetExplosion_Generic_PS))
  (Effects|Components|ParticleSystem|SetTemplate (Variables|Default|GetExplosion_Flesh_Damage_PSC) (Variables|Effects|GetExplosion_Flesh_Damage_PS))
  (Transformation|SetRelativeScale3D (Variables|Projectile_Rocket|GetMesh) (Variables|Projectile_Rocket|GetRocketScale))
  (Variables|Effects|SetRocketSpinRate (Math|Random|RandomFloatinRange (* (Variables|Effects|GetRocketSpinRate) -1) (Variables|Effects|GetRocketSpinRate))))
