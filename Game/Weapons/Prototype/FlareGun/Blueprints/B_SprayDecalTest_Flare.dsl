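; Prototype flare-gun spray decal actor that draws its decal on a box component, erodes it away and destroys itself.
; Construction randomizes the box's rotation and Y/Z size with fixed X depth, then feeds opacity, edge hardness, color and texture into a dynamic material.
; It also derives a clamped pre-fade delay from lifetime and fadeout time.
; BeginPlay waits that delay, then a rate-scaled timeline drives the Erode parameter until destruction.
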(event EventBeginPlay
  ;[1 1 1] Decal Lifetime
  (Utilities|FlowControl|Delay (Variables|DecalSetup|GetDecal%20RT%20Delay))
  (Components|Timeline|SetPlayRate (Variables|B_SprayDecalTest_Flare|GetFade%20Decal%20Out%20Timeline) (/ 1 (Variables|DecalSetup|GetDecal%20Fadeout%20Time)))
  (bind (_direction _fade) (Timeline|Fade%20Decal%20Out%20Timeline PlayFromStart)
    (:Update
      (Rendering|Material|SetScalarParameterValue (Variables|DecalSetup|GetDecal%20MID) "Erode" _fade :!class /Script/Engine.MaterialInstanceDynamic))
    (:Finished
      ;[1 0 0] Kill decal
      (Actor|DestroyActor))))

(fn ConstructionScript
  (bind (_returnvalue_x _returnvalue_y _returnvalue_z) (Math|Vector|BreakVector (Math|Conversions|ToVector(Float) (Math|Random|RandomFloatinRange (Variables|DecalSetup|GetDecal%20Scale%20Min) (Variables|DecalSetup|GetDecal%20Scale%20Max)))))
  ;[1 0.3496 0.04811] Transform Stuff
  (Variables|DecalSetup|SetDecal%20Location (Transformation|GetWorldLocation (Variables|Default|Get1x1x1BoxCenterAligned)))
  ;;[1 1 1] Random Size
  (Variables|SetDecal%20Scale_X (Math|Vector|MakeVector 16.0 _returnvalue_y _returnvalue_z))
  (bind _output_get (Variables|DecalSetup|SetDecal%20Transform (Math|Transform|MakeTransform (Variables|DecalSetup|GetDecal%20Location) (Math|Rotator|CombineRotators (Math|Rotator|MakeRotator (Math|Random|RandomFloatinRange 0.0 360) (Math|Random|RandomFloatinRange 0.0 360) (Math|Random|RandomFloatinRange 0.0 360)) (Variables|DecalSetup|GetDecal%20Rotation)) (Variables|DecalSetup|GetDecal%20Scale))))
  (Transformation|SetWorldTransform (Variables|Default|Get1x1x1BoxCenterAligned) _output_get false false)
  ;[0.02691 0.1823 1] Material Stuff
  (bind _returnvalue (Rendering|Material|CreateDynamicMaterialInstance (Variables|Default|Get1x1x1BoxCenterAligned) 0 (Variables|DecalSetup|GetDecal%20Material) "None" :!class /Script/Engine.PrimitiveComponent))
  (bind _output_get_2 (Variables|DecalSetup|SetDecal%20MID _returnvalue))
  (Rendering|Material|SetScalarParameterValue _output_get_2 "Fade" (Variables|DecalSetup|GetDecal%20Opacity) :!class /Script/Engine.MaterialInstanceDynamic)
  (Rendering|Material|SetScalarParameterValue (Variables|DecalSetup|GetDecal%20MID) "X Opacity Hardness" (Variables|DecalSetup|GetDecal%20X%20Opacity%20Hardness) :!class /Script/Engine.MaterialInstanceDynamic)
  (Rendering|Material|SetScalarParameterValue (Variables|DecalSetup|GetDecal%20MID) "Y Opacity Hardness" (Variables|DecalSetup|GetDecal%20Y%20Opacity%20Hardness) :!class /Script/Engine.MaterialInstanceDynamic)
  (Rendering|Material|SetScalarParameterValue (Variables|DecalSetup|GetDecal%20MID) "Z Falloff Hardness" (Variables|DecalSetup|GetDecal%20Z%20Falloff%20Hardness) :!class /Script/Engine.MaterialInstanceDynamic)
  (Rendering|Material|SetVectorParameterValue (Variables|DecalSetup|GetDecal%20MID) "Color" (Variables|DecalSetup|GetDecal%20Color) :!class /Script/Engine.MaterialInstanceDynamic)
  (Rendering|Material|SetTextureParameterValue (Variables|DecalSetup|GetDecal%20MID) "DecalTexture" (Variables|DecalSetup|GetDecal%20Texture))
  ;[0.3343 1 0.3646] Lifetime Settings
  (Variables|DecalSetup|SetDecal%20Fadeout%20Time (Math|Float|Clamp(Float) (Variables|DecalSetup|GetDecal%20Fadeout%20Time) 0.0 (Variables|DecalSetup|GetDecal%20Lifetime)))
  (Variables|DecalSetup|SetDecal%20RT%20Delay (Math|Float|Clamp(Float) (- (Variables|DecalSetup|GetDecal%20Lifetime) (Variables|DecalSetup|GetDecal%20Fadeout%20Time)) 0.0 100)))
