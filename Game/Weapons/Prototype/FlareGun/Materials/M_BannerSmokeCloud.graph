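# Translucent unlit smoke cloud material for the prototype flare gun, with depth test disabled and a Color parameter feeding emissive adjustments.
# It projects DecalTexture onto scene geometry like a decal, using the world position behind translucency relative to the object's pivot and scale.
# Per-axis sphere masks soften the volume's edges, then Fade scales and Erode subtracts from the result.
# Mobile uses a plain texture sample instead of the masked world-aligned projection.
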Material /Game/Weapons/Prototype/FlareGun/Materials/M_BannerSmokeCloud
graph {
 properties {bDisableDepthTest=true, blendMode=BLEND_Translucent, pixelDepthOffsetMode=PDOM_Legacy, shadingModel=MSM_Unlit, usedShadingModels=Unlit}
 nodes [
  n1 = MaterialFunctionCall() {materialFunction=&LocalPosition}
  n2 = MaterialFunctionCall() {materialFunction=/Engine/Functions/Engine_MaterialFunctions02/WorldPositionOffset/ObjectScale}
  Add(
   A=Subtract(A=Multiply(A=@n1."Local Position (Excluding Offsets)", B=@n2.Scale XYZ), B=WorldPosition().XYZ)
   B=MaterialFunctionCall() {materialFunction=&ObjectPivotPoint}.Object Pivot Location
  )
  n3 = MaterialFunctionCall() {materialFunction=/Engine/Functions/Engine_MaterialFunctions02/WorldPositionOffset/ComponentBasisVectors}
  n4 = Normalize(VectorInput=@n3.X)
  n5 = Subtract(
   A=MaterialFunctionCall() {materialFunction=&WorldPositionBehindTranslucency}.Result
   B=MaterialFunctionCall() {materialFunction=&ObjectPivotPoint}.Object Pivot Location
  )
  n6 = Normalize(VectorInput=@n3.Y)
  AppendVector(A=AppendVector(A=DotProduct(A=@n4, B=@n5), B=DotProduct(A=@n6, B=@n5)), B=DotProduct(A=Normalize(VectorInput=@n3.Z), B=@n5))
  ComponentMask(None=SceneTexture() {sceneTextureId=PPI_DecalMask}.Color) {r=true}
  MaterialFunctionCall(Emissive=VectorParameter() {parameterName=Color, defaultValue=<rgba 0 0 0 1>}.RGB) {materialFunction=&MF_FX_EmissiveAdjustments}
  n7 = Subtract(
   A=Divide(A=@n5, B=MaterialFunctionCall() {materialFunction=&ObjectScale}.Scale XYZ)
   B=Constant3Vector() {constant=<rgba -0.5 -0.5 -0.5 0>}
  ) {constB=0}
  n8 = MaterialFunctionCall(Float2=ComponentMask(None=@n7) {g=true, r=true}) {materialFunction=&BreakOutFloat2Components}
  n9 = Constant() {r=0.5}
  n10 = ScalarParameter() {defaultValue=0.8, parameterName=X Opacity Hardness}
  n11 = ScalarParameter() {defaultValue=0.8, parameterName=Y Opacity Hardness}
  n12 = Multiply(
   A=Multiply(A=SphereMask(A=@n8.R, B=@n9, Hardness=@n10) {attenuationRadius=0.5}, B=SphereMask(A=@n8.G, B=@n9, Hardness=@n11) {attenuationRadius=0.5})
   B=SphereMask(A=ComponentMask(None=@n7) {b=true}, B=Constant() {r=0.5}, Hardness=ScalarParameter() {defaultValue=0.8, parameterName=Z Falloff Hardness}) {attenuationRadius=0.5}
  )
  n13 = CameraPositionWS()
  n14 = WorldPosition()
  Multiply(A=@n12, B=OneMinus(None=SphereMask(A=@n13, B=@n14.XYZ) {hardnessPercent=0}))
  n15 = TextureObjectParameter() {
   parameterName=DecalTexture
   samplerType=SAMPLERTYPE_LinearColor
   texture=/Game/Effects/Fort_Effects/Gameplay/SmokeDustExplosions/Textures/T_GroundChar_01
  }
  n16 = ComponentMask(None=SceneTexture() {sceneTextureId=PPI_WorldNormal}.Color) {b=true, g=true, r=true}
  n17 = Constant3Vector() {constant=<rgba 0 0 1 0>}
  n18 = Multiply(
   A=MaterialFunctionCall!(
    Diffuse Texture Object=@n15
    Spec Texture Object
    TextureSize=Constant() {r=1}
    --------------------------
    Use seperate settings for Z projection
    Use seperate texture for Z projection
    Texture Size Z projection
    Z Plane Diffuse Texture Object
    Z Plane Spec Texture Object
    "--------------------------      "
    ProjectionTransitionContrast
    Use Texture Alpha Heightmap to Refine the Transistion
    HeightLerpTransitionContrast
    PickTheDiffuseHeightChannel
    TheDiffuseHeightChannel
    "--------------------------       "
    NormalTextureObject
    World Space Normals
    "--------------------------  "
    Export Float 4 Diffuse
    Export Float 4 Spec
    "--------------------------     "
    WorldPosition=@n7
    World Space Normal=ShadingPathSwitch(Default=@n16, Deferred=@n16, Forward=@n17, Mobile=@n17)
   ) {materialFunction=/Engine/Functions/Engine_MaterialFunctions02/Texturing/WorldAlignedTextures_Complex}.XYZ Diffuse Texture
   B=Multiply(A=@n12)
  )
  n19 = TextureSample(Tex=@n15)
  Subtract(
   A=LinearInterpolate(
    B=ShadingPathSwitch(Default=@n18, Mobile=@n19.RGB) {desc=Migrated from FeatureLevelSwitch}
    Alpha=ScalarParameter() {defaultValue=0.9, parameterName=Fade}
   ) {constB=0}
   B=ScalarParameter() {parameterName=Erode}
  )
 ]
}
